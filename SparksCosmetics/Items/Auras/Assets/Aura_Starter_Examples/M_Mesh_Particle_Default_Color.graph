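# Translucent, unlit, two-sided material for Niagara mesh particles and sprites, stored among the Sparks cosmetics aura starter examples.
# Its small graph feeds Particle Color alpha into a Multiply node, with no material output connections shown.
# It allows custom depth writes despite being translucent.
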Material /SparksCosmetics/Items/Auras/Assets/Aura_Starter_Examples/M_Mesh_Particle_Default_Color
graph {
 properties {
  allowTranslucentCustomDepthWrites=true
  bUsedWithNiagaraMeshParticles=true
  bUsedWithNiagaraSprites=true
  blendMode=BLEND_Translucent
  pixelDepthOffsetMode=PDOM_Legacy
  shadingModel=MSM_Unlit
  twoSided=true
  usedShadingModels=Unlit
 }
 nodes [Multiply(A=ParticleColor().A)]
}
